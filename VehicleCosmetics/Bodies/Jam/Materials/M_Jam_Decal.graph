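# M_Jam_Decal is a masked, two-sided Default Lit material, named as a decal for the Jam vehicle cosmetic body.
# Its graph holds a single texture sample of the T_Jam_Logo_D logo texture, with no connection to material outputs shown.
# It supports skeletal meshes and mesh deformers, with automatic usage flagging and ray-traced shadow casting turned off.
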Material /VehicleCosmetics/Bodies/Jam/Materials/M_Jam_Decal
graph {
 properties {
  bAutomaticallySetUsageInEditor=false
  bCastRayTracedShadows=false
  bUsedWithMeshDeformer=true
  bUsedWithSkeletalMesh=true
  blendMode=BLEND_Masked
  twoSided=true
  usedShadingModels=Default Lit
 }
 nodes [TextureSample() {texture=/VehicleCosmetics/Bodies/Jam/Textures/T_Jam_Logo_D}]
}
