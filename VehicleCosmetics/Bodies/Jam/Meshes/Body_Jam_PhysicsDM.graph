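# Physics asset for the Jam vehicle body mesh in VehicleCosmetics, giving the cosmetic body a simplified collision shape.
# It holds a single body on the root bone made of one box primitive, offset slightly forward and raised to enclose the car.
# No constraints are defined, so the body simulates as one rigid piece.
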PhysicsAsset /VehicleCosmetics/Bodies/Jam/Meshes/Body_Jam_PhysicsDM
bodies {root [{shapeType=Box, extentX=444, extentY=216, extentZ=116, shapeName=root_box, center=<20.3382 0 95.3351>}]}
constraints []
